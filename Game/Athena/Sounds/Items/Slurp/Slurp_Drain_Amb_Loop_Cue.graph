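# Sound cue for the looping ambient sound of a Slurp drain, classed as a Fort consumable.
# A Random node picks one of two looping wave variants with equal weight, and a Modulator lowers its pitch by a random amount for variation.
# It uses the consumable ambient attenuation and overrides concurrency to cap simultaneous instances, stopping the oldest first.
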SoundCue /Game/Athena/Sounds/Items/Slurp/Slurp_Drain_Amb_Loop_Cue
root=Modulator_0
nodes [
 {id=Modulator_0, children [Random_0], properties {pitchMax=1, pitchMin=0.8}}
 {id=Random_0, children [WavePlayer_0,WavePlayer_1], properties {weights [1,1]}}
 {id=WavePlayer_0, properties {bLooping=true, soundWaveAssetPtr=&/Item_Slurp_Drain_Amb_Loop_01}}
 {id=WavePlayer_1, properties {bLooping=true, soundWaveAssetPtr=&/Item_Slurp_Drain_Amb_Loop_02}}
]
properties {
 attenuationSettings=/Game/Sounds/Attenuation/Fort_Consumable_Amb_Att
 bOverrideConcurrency=true
 maxDistance=2150
 soundClassObject=&Fort_Consumables
 volumeMultiplier=1
 concurrencyOverrides {maxCount=4, resolutionRule=StopOldest}
}
