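# Sound cue for the grab sound of a Slurp mushroom consumable in Athena, reusing the mushroom shield-increase grab wave.
# A single wave player feeds a modulator node left at its default settings.
# The cue plays in the consumables sound class with consumable-use attenuation and a boosted volume.
# A concurrency override caps simultaneous instances, stopping the oldest when the limit is exceeded.
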SoundCue /Game/Athena/Sounds/Items/Slurp/Slurp_Shroom_Grab_Cue
root=Modulator_0
nodes [
 {id=Modulator_0, children [WavePlayer_0]}
 {id=WavePlayer_0, properties {soundWaveAssetPtr=/Game/Sounds/Fort_GamePlay_Sounds/Player/Mushroom_Shield_Increase_Grab}}
]
properties {
 attenuationSettings=&Fort_Consumable_Use_Athena_Att
 bOverrideConcurrency=true
 maxDistance=2250
 soundClassObject=&Fort_Consumables
 volumeMultiplier=1.5
 concurrencyOverrides {maxCount=4, resolutionRule=StopOldest}
}
