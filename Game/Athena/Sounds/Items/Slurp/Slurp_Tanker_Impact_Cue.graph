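# Sound cue for impacts on the Slurp tanker, played as a spatialized Fort_Impacts sound.
# A Random node picks one of six equally weighted Slurp tower impact waves, and a Modulator above it varies pitch while holding volume constant.
# The cue references a small-object demolition-hit attenuation asset but enables its own attenuation override, plus a concurrency override that prevents new instances at the limit.
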SoundCue /Game/Athena/Sounds/Items/Slurp/Slurp_Tanker_Impact_Cue
root=Modulator_12
nodes [
 {id=Modulator_12, children [Random_0], properties {pitchMin=0.9, volumeMax=1, volumeMin=1}}
 {
  id=Random_0
  children [WavePlayer_1, WavePlayer_2, WavePlayer_3, WavePlayer_4, WavePlayer_10, WavePlayer_11]
  properties {weights [1,1,1,1,1,1]}
 }
 {id=WavePlayer_1, properties {soundWaveAssetPtr=&/Item_Slurp_Tower_Impact_01}}
 {id=WavePlayer_10, properties {soundWaveAssetPtr=&/Item_Slurp_Tower_Impact_05}}
 {id=WavePlayer_11, properties {soundWaveAssetPtr=&/Item_Slurp_Tower_Impact_06}}
 {id=WavePlayer_2, properties {soundWaveAssetPtr=&/Item_Slurp_Tower_Impact_02}}
 {id=WavePlayer_3, properties {soundWaveAssetPtr=&/Item_Slurp_Tower_Impact_03}}
 {id=WavePlayer_4, properties {soundWaveAssetPtr=&/Item_Slurp_Tower_Impact_04}}
]
properties {
 attenuationSettings=/Game/Sounds/Attenuation/Fort_Demolition_Hit_Small_Sized_Object_Att
 bOverrideAttenuation=true
 bOverrideConcurrency=true
 maxDistance=3000
 soundClassObject=&Fort_Impacts
 volumeMultiplier=0.96
 attenuationOverrides {falloffDistance=2400, attenuationShapeExtents {x=600}}
 concurrencyOverrides {resolutionRule=PreventNew}
}
