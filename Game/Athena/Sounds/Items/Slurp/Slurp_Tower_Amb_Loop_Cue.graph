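# Sound cue for the looping ambient sound of the Slurp Tower item in Athena.
# A Random node picks one of two looping wave variants with equal weight, then a Modulator varies its pitch.
# The cue uses consumable ambient attenuation and the consumables sound class, with boosted volume.
# A concurrency override caps simultaneous instances at a few, stopping the oldest first.
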SoundCue /Game/Athena/Sounds/Items/Slurp/Slurp_Tower_Amb_Loop_Cue
root=Modulator_0
nodes [
 {id=Modulator_0, children [Random_1], properties {pitchMin=0.9}}
 {id=Random_1, children [WavePlayer_1,WavePlayer_3], properties {weights [1,1]}}
 {id=WavePlayer_1, properties {bLooping=true, soundWaveAssetPtr=&/Item_Slurp_Tower_Amb_Loop_01}}
 {id=WavePlayer_3, properties {bLooping=true, soundWaveAssetPtr=&/Item_Slurp_Tower_Amb_Loop_02}}
]
properties {
 attenuationSettings=/Game/Sounds/Attenuation/Fort_Consumable_Amb_Att
 bOverrideConcurrency=true
 maxDistance=2150
 soundClassObject=&Fort_Consumables
 volumeMultiplier=2.5
 concurrencyOverrides {maxCount=4, resolutionRule=StopOldest}
}
