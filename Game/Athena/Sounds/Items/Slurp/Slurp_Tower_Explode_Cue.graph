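# Sound cue for a Slurp tower explosion in Athena, routed to the impacts sound class with boosted volume.
# A random node picks one of four equally weighted explosion waves, and a modulator varies its pitch while holding volume steady.
# The cue overrides attenuation with its own falloff despite referencing a smoke grenade preset.
# Its concurrency override prevents new instances once the limit is reached.
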SoundCue /Game/Athena/Sounds/Items/Slurp/Slurp_Tower_Explode_Cue
root=Modulator_12
nodes [
 {id=Modulator_12, children [Random_1], properties {pitchMin=0.9, volumeMax=1, volumeMin=1}}
 {id=Random_1, children [WavePlayer_0,WavePlayer_5,WavePlayer_6,WavePlayer_7], properties {weights [1,1,1,1]}}
 {id=WavePlayer_0, properties {soundWaveAssetPtr=&/Item_Slurp_Tower_Explode_01}}
 {id=WavePlayer_5, properties {soundWaveAssetPtr=&/Item_Slurp_Tower_Explode_02}}
 {id=WavePlayer_6, properties {soundWaveAssetPtr=&/Item_Slurp_Tower_Explode_03}}
 {id=WavePlayer_7, properties {soundWaveAssetPtr=&/Item_Slurp_Tower_Explode_04}}
]
properties {
 attenuationSettings=/Game/Sounds/Attenuation/Fort_Grenade_Smoke_Att
 bOverrideAttenuation=true
 bOverrideConcurrency=true
 maxDistance=3000
 soundClassObject=&Fort_Impacts
 volumeMultiplier=1.5
 attenuationOverrides {falloffDistance=2400, attenuationShapeExtents {x=600}}
 concurrencyOverrides {resolutionRule=PreventNew}
}
