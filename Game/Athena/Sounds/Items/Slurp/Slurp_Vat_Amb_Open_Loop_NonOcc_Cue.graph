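# Sound cue for the looping ambience of an open Slurp vat item, in a non-occluded variant.
# A Random node picks one of two equally weighted looping waves, then a Modulator randomly varies its pitch from normal to slightly higher.
# It uses non-occluded consumable ambience attenuation and the consumables sound class, with a concurrency override that caps simultaneous instances by stopping the oldest.
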SoundCue /Game/Athena/Sounds/Items/Slurp/Slurp_Vat_Amb_Open_Loop_NonOcc_Cue
root=Modulator_0
nodes [
 {id=Modulator_0, children [Random_0], properties {pitchMax=1.15, pitchMin=1}}
 {id=Random_0, children [WavePlayer_0,WavePlayer_2], properties {weights [1,1]}}
 {id=WavePlayer_0, properties {bLooping=true, soundWaveAssetPtr=&/Item_Slurp_Vat_Amb_Loop_Open_01}}
 {id=WavePlayer_2, properties {bLooping=true, soundWaveAssetPtr=&/Item_Slurp_Vat_Amb_Loop_Open_02}}
]
properties {
 attenuationSettings=/Game/Sounds/Attenuation/Fort_Consumable_Amb_NonOccluded_Att
 bOverrideConcurrency=true
 maxDistance=1750
 soundClassObject=&Fort_Consumables
 volumeMultiplier=1.2
 concurrencyOverrides {maxCount=4, resolutionRule=StopOldest}
}
