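# Sound cue for impacts on a Slurp vat, playing one of six lid impact variations.
# An equally weighted random node picks the wave and feeds a modulator that randomizes pitch at fixed volume.
# It uses the impacts sound class and a small-object demolition hit attenuation asset, with cue-level overrides for attenuation and for concurrency that prevents new instances at the limit.
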SoundCue /Game/Athena/Sounds/Items/Slurp/Slurp_Vat_Impact_Cue
root=Modulator_12
nodes [
 {id=Modulator_12, children [Random_1], properties {pitchMin=0.9, volumeMax=1, volumeMin=1}}
 {id=Random_1, children [WavePlayer_0,WavePlayer_5,WavePlayer_6,WavePlayer_7,WavePlayer_8,WavePlayer_9], properties {weights [1,1,1,1,1,1]}}
 {id=WavePlayer_0, properties {soundWaveAssetPtr=&/Item_Slurp_Vat_Impact_Lid_01}}
 {id=WavePlayer_5, properties {soundWaveAssetPtr=&/Item_Slurp_Vat_Impact_Lid_02}}
 {id=WavePlayer_6, properties {soundWaveAssetPtr=&/Item_Slurp_Vat_Impact_Lid_03}}
 {id=WavePlayer_7, properties {soundWaveAssetPtr=&/Item_Slurp_Vat_Impact_Lid_04}}
 {id=WavePlayer_8, properties {soundWaveAssetPtr=&/Item_Slurp_Vat_Impact_Lid_05}}
 {id=WavePlayer_9, properties {soundWaveAssetPtr=&/Item_Slurp_Vat_Impact_Lid_06}}
]
properties {
 attenuationSettings=/Game/Sounds/Attenuation/Fort_Demolition_Hit_Small_Sized_Object_Att
 bOverrideAttenuation=true
 bOverrideConcurrency=true
 maxDistance=3000
 soundClassObject=&Fort_Impacts
 volumeMultiplier=0.96
 attenuationOverrides {falloffDistance=2400, attenuationShapeExtents {x=600}}
 concurrencyOverrides {resolutionRule=PreventNew}
}
